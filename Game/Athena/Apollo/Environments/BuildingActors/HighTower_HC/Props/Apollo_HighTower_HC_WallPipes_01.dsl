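; Apollo HighTower HC wall pipes prop, a placeable building actor with overrides for creative-mode placement queries.
; The bounds, origin and relative transform queries return false with zero or default values.
; The grid axis guide attachment query returns true with a fixed offset point.
; The construction script is empty.
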(fn GetCreativeActorBounds
  (return false "0, 0, 0"))

(fn GetCreativeActorOrigin
  (return false "0, 0, 0"))

(fn GetCreativeActorRelativeTransform
  (return false 0))

(fn GetGridAxisGuideAttachmentPoint
  (return true false "-230.000000,38.000000,190.000000"))

(fn ConstructionScript)
